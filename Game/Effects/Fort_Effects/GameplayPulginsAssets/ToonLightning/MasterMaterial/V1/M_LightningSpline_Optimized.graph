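# Translucent unlit Niagara master material for toon lightning bolts, with a static switch between bolt and glow looks.
# Vertex offsets read baked position, tangent and thickness texture arrays by particle index to bend the strip and widen it facing the camera.
# Dynamic parameters drive growth, width, dissolve and alignment, while textures and edge masks shape color and opacity.
# Noise and move-to-center offsets are zeroed on mobile, Switch and performance mode.
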Material /Game/Effects/Fort_Effects/GameplayPulginsAssets/ToonLightning/MasterMaterial/V1/M_LightningSpline_Optimized
graph {
 properties {
  bUsedWithNiagaraMeshParticles=true
  bUsedWithNiagaraRibbons=true
  bUsedWithNiagaraSprites=true
  blendMode=BLEND_Translucent
  pixelDepthOffsetMode=PDOM_Legacy
  shadingModel=MSM_Unlit
  usedShadingModels=Unlit
 }
 nodes [
  n1 = NamedRerouteUsage()
  n2 = NamedRerouteUsage()
  Add(A=Add(A=NamedRerouteUsage(), B=Add(A=@n2, B=@n1)), B=NamedRerouteUsage())
  n3 = MaterialFunctionCall(Float2=NamedRerouteUsage()) {materialFunction=&BreakOutFloat2Components}
  n4 = VectorParameter() {group=SmpGlow, parameterName=SmpGlow_EdgeMask_Control, channelNames {a="", b=Power, g=Multiply, r=Offset}, defaultValue=<rgba -1 1 1 1>}
  n5 = NamedRerouteUsage()
  n6 = VectorParameter() {group=em, parameterName=Glow_EdgeMask_Control, channelNames {a="", b=Power, g=Multiply, r=Offset}, defaultValue=<rgba 0 2 2 1>}
  n7 = VectorParameter() {parameterName=Glow_Refine, defaultValue=<rgba 4 2 0.5 0.5>}
  n8 = MaterialFunctionCall(
   Texture=MaterialFunctionCall!(
    UorV=MaterialFunctionCall(Float2=@n5) {materialFunction=&BreakOutFloat2Components}.R
    EdgeControl=@n6
   ) {materialFunction=/Game/Effects/Niagara/MasterMaterials/Functions/Common/FXCommon_EdgeMask_Simple}.Mask
   RGB=@n7.RGB
   A=@n7.RGB
  ) {materialFunction=&MF_FX_Refine}
  n9 = NamedRerouteUsage()
  n10 = VectorParameter() {
   group=UPDOWNFADE
   parameterName=GlowFade_EdgeMask_InvertControl
   channelNames {a="", b=Power, g=Multiply, r=Offset}
   defaultValue=<rgba 0 20 2 1>
  }
  n11 = VectorParameter() {group=UPDOWNFADE, parameterName=GlowFade_EdgeMask_Control, channelNames {a="", b=Power, g=Multiply, r=Offset}, defaultValue=<rgba 0 2 1 1>}
  n12 = Multiply(
   A=@n8.Result
   B=MaterialFunctionCall!(
    UorV=MaterialFunctionCall(Float2=@n9) {materialFunction=&BreakOutFloat2Components}.G
    Invert_EdgeControl=@n10
    NoInvert_EdgeControl=@n11
   ) {materialFunction=/Game/Effects/Niagara/MasterMaterials/Functions/Common/FXCommon_EdgeMask_Ex}.Mask
  )
  n13 = TextureObjectParameter() {
   group=LightningT
   parameterName=LightningT_Texture
   sortPriority=0
   texture=&/T_LightningSplineTexture
  }
  n14 = DynamicParameter() {
   parameterIndex=2
   defaultValue=<rgba 5 0 500 30>
   paramNames [ScaleSplineWidth, BaseWitdthOffset, TextureRandom, MoveAlignLightningDirSphereMaskSize]
  }
  n15 = VectorParameter() {group=LightningT, parameterName=LightningT_UVControl, channelNames {a=VPan, b=UPan, g=VScale, r=UScale}, defaultValue=<rgba 0.5 1.5 0 0>}
  n16 = MaterialFunctionCall!(
   Texture=@n13
   AdditionalUVOffset=MaterialFunctionCall(X=Constant(), Y=@n14.TextureRandom) {materialFunction=&MakeFloat2}.Result
   UVControl_RGB=@n15
   UVControl_A=@n15
   "[Override]TimeInput"
   "[Override]ScaleUVByCenter"=StaticBool() {value=true}
   "[Override]UVInput"=NamedRerouteUsage()
  ) {materialFunction=/Game/Effects/Niagara/MasterMaterials/Functions/Common/FXCommon_TextureSample_V2}
  n17 = VectorParameter() {parameterName=OutGlowRefine, defaultValue=<rgba 1.5 0.5 0 0>}
  n18 = MaterialFunctionCall(Texture=@n16.R, RGB=@n17.RGB, A=@n17.RGB) {materialFunction=&MF_FX_Refine}
  n19 = MaterialFunctionCall(Float2=NamedRerouteUsage()) {materialFunction=&BreakOutFloat2Components}
  n20 = LinearInterpolate(A=@n19.G, B=OneMinus(None=@n19.G), Alpha=ScalarParameter() {defaultValue=1, parameterName=SplineThickness_GrowthDirectionInvert})
  n21 = DynamicParameter() {
   parameterIndex=3
   defaultValue=<rgba 0 0 1 -0.7>
   paramNames [MoveAlignLightningDir, "MoveToCenter(NoPerf/Mobile)", Growth, ThicknessDissolve]
  }
  n22 = Saturate(None=@n21.Growth)
  n23 = DynamicParameter() {paramNames [MaxGrowthLength, NewLightningTech_InvertDir, WPORand, NewLightningTech_Index]}
  n24 = ScalarParameter() {defaultValue=5, parameterName=SplineThickness_GrowthGradient}
  n25 = Divide(B=@n24)
  n26 = Add(A=Subtract(A=@n20), B=Multiply(A=Multiply(A=@n22, B=@n23.MaxGrowthLength), B=Add(B=@n25) {constA=1}))
  n27 = Add(A=@n20, B=@n21.Growth)
  n28 = StaticSwitchParameter(True=Saturate(None=Multiply(A=@n26, B=@n24)), False=Saturate(None=Multiply(A=@n27, B=@n24))) {parameterName=UseNewGrowthMethod}
  n29 = Add(
   A=Multiply(
    A=MaterialFunctionCall(UorV=@n3.R, EdgeControl=@n4) {materialFunction=&FXCommon_EdgeMask_Simple}.Mask
    B=@n4.A
   )
   B=Multiply(A=@n28, B=StaticSwitchParameter(True=@n12, False=@n18.Result) {parameterName=Glow})
  ) {constB=0}
  n30 = VectorParameter() {
   group=EdgeFade
   parameterName=EdgeFade_EdgeMask_InvertControl
   channelNames {a="", b=Power, g=Multiply, r=Offset}
   defaultValue=<rgba 5 1 1 1>
  }
  n31 = VectorParameter() {group=EdgeFade, parameterName=EdgeFade_EdgeMask_Control, channelNames {a="", b=Power, g=Multiply, r=Offset}, defaultValue=<rgba 5 1 1 1>}
  n32 = LinearInterpolate(
   B=MaterialFunctionCall(UorV=@n3.G, Invert_EdgeControl=@n30, NoInvert_EdgeControl=@n31) {materialFunction=&FXCommon_EdgeMask_Ex}.Mask
   Alpha=ScalarParameter() {defaultValue=1, group=EdgeFade, parameterName=Use edge fade mask Control}
  ) {constA=1}
  n33 = Multiply(A=NamedRerouteUsage(), B=DepthFade(Opacity=Multiply(A=@n29, B=@n32), FadeDistance=ScalarParameter() {parameterName=DF}))
  n34 = ParticleColor()
  n35 = ScalarParameter() {parameterName=NearCameraFade}
  n36 = Multiply(
   A=DepthFade(Opacity=Multiply(A=@n33, B=@n34.A), FadeDistance=ScalarParameter() {defaultValue=1, parameterName=DepthFade})
   B=Power(
    Base=MaterialFunctionCall(Distance=@n35) {materialFunction=&Near_Camera_Fade}.Result
    Exp=ScalarParameter() {defaultValue=2, parameterName=NearCameraFadePower}
   )
  )
  LinearInterpolate(A=@n36, B=Clamp(None=@n36) {maxDefault=0.999}, Alpha=ScalarParameter() {parameterName=Limit Alpha less than 1})
  n37 = MaterialFunctionCall(Float2=NamedRerouteUsage()) {materialFunction=&BreakOutFloat2Components}
  n38 = VectorParameter() {parameterName=BottomBright_Refine, defaultValue=<rgba 4 10 0 0.5>}
  n39 = MaterialFunctionCall(Texture=@n37.G, RGB=@n38.RGB, A=@n38.RGB) {materialFunction=&MF_FX_Refine}
  n40 = VectorParameter() {parameterName=TipBright_Refine, defaultValue=<rgba 4 10 0 0.5>}
  n41 = Add(
   A=@n39.Result
   B=MaterialFunctionCall(Texture=OneMinus(None=@n37.G), RGB=@n40.RGB, A=@n40.RGB) {materialFunction=&MF_FX_Refine}.Result
  )
  n42 = ScalarParameter() {parameterName=UseTipBright}
  n43 = ScalarParameter() {defaultValue=0.1, parameterName=SpikeBrightness_SpikeDataSubtract}
  n44 = Saturate(None=Subtract(A=NamedRerouteUsage(), B=@n43))
  n45 = OneMinus(None=@n43)
  n46 = Multiply(
   A=Power(Base=Divide(A=@n44, B=@n45), Exp=ScalarParameter() {defaultValue=10, parameterName=SpikeBrightness_SpikeDataPower})
   B=ScalarParameter() {parameterName=SpikeBrightness_Scale}
  )
  n47 = VectorParameter() {parameterName=Glow_ColorA, defaultValue=<rgba 0.004 0.053 0.2 1>}
  n48 = VectorParameter() {parameterName=Glow_ColorB, defaultValue=<rgba 0.04 0.13 0.4 1>}
  n49 = NamedRerouteUsage()
  n50 = Power(Base=@n49, Exp=ScalarParameter() {defaultValue=3, parameterName=Glow_ColorFromSpike_Power})
  n51 = Power(Base=@n49, Exp=ScalarParameter() {defaultValue=2, parameterName=Glow_BrogjtmessFromSpike_Power})
  n52 = ScalarParameter() {defaultValue=0.1, parameterName=Glow_BrogjtmessFromSpike_Str}
  n53 = Multiply(A=Multiply(A=Add(A=@n8.Result), B=LinearInterpolate(A=@n47.RGB, B=@n48.RGB, Alpha=@n50)), B=Add(A=Multiply(A=@n51, B=@n52)))
  n54 = VectorParameter() {parameterName=InnerGlowRefine, defaultValue=<rgba 5 2 1 0.5>}
  n55 = Multiply(
   A=MaterialFunctionCall(Texture=OneMinus(None=@n16.G), RGB=@n54.RGB, A=@n54.RGB) {materialFunction=&MF_FX_Refine}.Result
   B=@n16.B
  )
  n56 = LinearInterpolate(
   A=VectorParameter() {parameterName=GlowColor, defaultValue=<rgba 0 0.8333 5 1>}.RGB
   B=LinearInterpolate(
    A=VectorParameter() {parameterName=BoltColor, defaultValue=<rgba 0 0.8333 1.25 1>}.RGB
    B=VectorParameter() {parameterName=InnerGlowColor, defaultValue=<rgba 0.65 4.7667 13 5>}.RGB
    Alpha=Saturate(None=@n55)
   )
   Alpha=@n16.B
  )
  n57 = Multiply(
   A=Add(A=LinearInterpolate(A=@n39.Result, B=@n41, Alpha=@n42), B=ScalarParameter() {defaultValue=1, parameterName=BottomBright_Base})
   B=Multiply(A=Add(A=@n46), B=StaticSwitchParameter(True=@n53, False=@n56) {parameterName=Glow})
  )
  n58 = Constant3Vector() {constant=<rgba 1 0 0 0>}
  n59 = NamedRerouteUsage()
  n60 = DynamicParameter() {
   parameterIndex=3
   defaultValue=<rgba 0 0 1 -0.7>
   paramNames [MoveAlignLightningDir, "MoveToCenter(NoPerf/Mobile)", Growth, ThicknessDissolve]
  }
  n61 = Saturate(None=Multiply(A=Subtract(A=@n59, B=@n60.ThicknessDissolve), B=ScalarParameter() {defaultValue=2, parameterName=ThicknessDissolveHardness}))
  n62 = Floor(None=@n61)
  n63 = Multiply(
   A=LinearInterpolate(A=@n57, B=LinearInterpolate(A=@n58, B=@n57, Alpha=@n62), Alpha=ScalarParameter() {parameterName=DebugThicknessDissolve})
   B=@n34.RGB
  )
  n64 = AppendVector(A=TextureCoordinate(), B=Constant())
  n65 = ScalarParameter() {parameterName=DebugUV}
  MaterialFunctionCall(Emissive=Divide(A=LinearInterpolate(A=@n63, B=@n64, Alpha=@n65)) {constB=2.1}) {materialFunction=&MF_FX_EmissiveAdjustments}
  n66 = WorldPosition() {worldPositionShaderOffset=WPT_ExcludeAllShaderOffsets}
  n67 = MaterialFunctionCall(Float3=Add(
    A=TransformPosition(None=@n66.XYZ) {transformSourceType=TRANSFORMPOSSOURCE_World, transformType=TRANSFORMPOSSOURCE_Instance}
    B=NamedRerouteUsage()
   )) {materialFunction=&BreakOutFloat3Components}
  n68 = MaterialFunctionCall(X=@n67.R, Y=@n67.R) {materialFunction=&MakeFloat2}
  n69 = Constant2Vector()
  n70 = DynamicParameter() {
   parameterIndex=2
   defaultValue=<rgba 5 0 500 30>
   paramNames [ScaleSplineWidth, BaseWitdthOffset, TextureRandom, MoveAlignLightningDirSphereMaskSize]
  }
  n71 = DynamicParameter() {
   parameterIndex=1
   defaultValue=<rgba 0 2 -1 0>
   paramNames [NewLightningTech_WPOStr, MoveAlignLightningDirSphereMaskPower, TextureDissolve, TextureWPONoise]
  }
  n72 = Power(
   Base=OneMinus(None=SphereMask(A=@n68.Result, B=@n69, Radius=@n70.MoveAlignLightningDirSphereMaskSize) {attenuationRadius=1, hardnessPercent=0})
   Exp=@n71.MoveAlignLightningDirSphereMaskPower
  )
  n73 = DynamicParameter() {
   parameterIndex=3
   defaultValue=<rgba 0 0 1 -0.7>
   paramNames [MoveAlignLightningDir, "MoveToCenter(NoPerf/Mobile)", Growth, ThicknessDissolve]
  }
  n74 = Constant3Vector() {constant=<rgba 0 0 1 0>}
  NamedRerouteDeclaration(None=Multiply(A=Multiply(A=@n72, B=@n73.MoveAlignLightningDir), B=Transform(None=@n74) {transformSourceType=TRANSFORMSOURCE_Instance})) {name=WPO_MoveAlignLightningDir, nodeColor=<rgba 1 0 0.3047 1>}
  n75 = Subtract(
   A=Constant3Vector()
   B=Add(
    A=MaterialFunctionCall() {materialFunction=&LocalPosition}.Instance Local Position
    B=NamedRerouteUsage()
   )
  )
  n76 = DynamicParameter() {
   parameterIndex=3
   defaultValue=<rgba 0 0 1 -0.7>
   paramNames [MoveAlignLightningDir, "MoveToCenter(NoPerf/Mobile)", Growth, ThicknessDissolve]
  }
  n77 = Transform(None=Multiply(A=Multiply(A=@n75, B=@n76."MoveToCenter(NoPerf/Mobile)"), B=Constant3Vector() {constant=<rgba 1 1 0 0>})) {transformSourceType=TRANSFORMSOURCE_Local}
  n78 = Constant3Vector()
  NamedRerouteDeclaration(None=MaterialFunctionCall!("Xb1,PS4"=@n77, "PC-High&&Epic,XSX,PS5"=@n77, "PC-Perf,Android"=@n78, Switch=@n78, "PC-Low(NotPerf)&&Medium"=@n77) {materialFunction=/Game/Effects/Niagara/MasterMaterials/Functions/Common/PlatformSwitch_3D}.Result) {name=WPO_MoveToCenter, nodeColor=<rgba 1 0.7987 0 1>}
  n79 = OneMinus(None=MaterialFunctionCall(Float2=TextureCoordinate()) {materialFunction=&BreakOutFloat2Components}.G)
  n80 = Floor(None=Add(
    A=DynamicParameter() {paramNames [MaxGrowthLength, NewLightningTech_InvertDir, WPORand, NewLightningTech_Index]}.NewLightningTech_Index
    B=ScalarParameter() {group=NewLightningTech, parameterName=NewLightningTech_DebugIndex}
   ))
  n81 = TextureSampleParameter2DArray!(
   UVs=MaterialFunctionCall(R=@n79, G=Constant() {r=0.5}, B=@n80) {materialFunction=&AppendMany}.RGB
   Apply View MipBias
  ) {
   group=NewLightningTech
   parameterName=LightningTexture_PositionArray
   samplerType=SAMPLERTYPE_LinearColor
   texture=/Game/Effects/Fort_Effects/GameplayPulginsAssets/ToonLightning/GP29/Texture/T_GP29Lightning_PositionArray
  }
  n82 = MaterialFunctionCall(Float3=Multiply(A=Subtract(A=@n81.RGB) {constB=0.5}) {constB=2}) {materialFunction=&BreakOutFloat3Components}
  n83 = DynamicParameter() {paramNames [MaxGrowthLength, NewLightningTech_InvertDir, WPORand, NewLightningTech_Index]}
  n84 = LinearInterpolate(Alpha=Saturate(None=Floor(None=@n83.NewLightningTech_InvertDir))) {constA=1, constB=-1}
  n85 = ScalarParameter() {defaultValue=1, group=NewLightningTech, parameterName=NewLightningTech_ScaleZ}
  n86 = DynamicParameter() {
   parameterIndex=1
   defaultValue=<rgba 0 2 -1 0>
   paramNames [NewLightningTech_WPOStr, MoveAlignLightningDirSphereMaskPower, TextureDissolve, TextureWPONoise]
  }
  n87 = Multiply(
   A=MaterialFunctionCall(X=Multiply(A=@n82.R, B=@n84), Y=Multiply(A=@n82.G, B=@n84), Z=Divide(A=@n82.B, B=@n85)) {materialFunction=&MakeFloat3}.Result
   B=Multiply(A=ScalarParameter() {defaultValue=20, group=NewLightningTech, parameterName=NewLightningTech_WPOStr}, B=@n86.NewLightningTech_WPOStr)
  )
  n88 = Transform(None=@n87) {transformSourceType=TRANSFORMSOURCE_Instance}
  n89 = Constant() {r=1}
  n90 = TextureSampleParameter2DArray(UVs=MaterialFunctionCall(R=@n79, G=Constant() {r=0.5}, B=@n80) {materialFunction=&AppendMany}.RGB) {
   group=NewLightningTech
   parameterName=LightningTexture_TangentArray
   samplerType=SAMPLERTYPE_LinearColor
   texture=/Game/Effects/Fort_Effects/GameplayPulginsAssets/ToonLightning/GP29/Texture/T_GP29Lightning_TangentArray
  }
  n91 = Subtract(A=@n90.RGB) {constB=0.5}
  n92 = Transform(None=Multiply(
    A=MaterialFunctionCall(X=@n84, Y=@n84, Z=@n89) {materialFunction=&MakeFloat3}.Result
    B=Multiply(A=@n91) {constB=2}
   )) {transformSourceType=TRANSFORMSOURCE_Instance}
  n93 = Subtract(A=CameraPositionWS(), B=Add(A=WorldPosition().XYZ, B=@n88))
  n94 = MaterialFunctionCall() {materialFunction=&TangentBasis}
  n95 = LinearInterpolate(
   A=Normalize(VectorInput=CrossProduct(A=@n92, B=@n93))
   B=Normalize(VectorInput=CrossProduct(A=@n94.Y V, B=@n93))
   Alpha=ScalarParameter() {group=NewLightningTech, parameterName=NewLightningTech_TangentBlend}
  )
  n96 = ScalarParameter() {defaultValue=100, parameterName=SplineThickness_BaseWidth}
  n97 = DynamicParameter() {
   parameterIndex=2
   defaultValue=<rgba 5 0 500 30>
   paramNames [ScaleSplineWidth, BaseWitdthOffset, TextureRandom, MoveAlignLightningDirSphereMaskSize]
  }
  n98 = StaticSwitchParameter(True=@n96, False=Max(B=Add(A=@n97.BaseWitdthOffset, B=@n96))) {parameterName=Glow}
  n99 = ScalarParameter() {defaultValue=600, parameterName=SplineThickness_SpikeWidth}
  n100 = ScalarParameter() {defaultValue=0.1, parameterName=SplineThickness_SpikeDataSubtract}
  n101 = Subtract(A=NamedRerouteUsage(), B=@n100)
  n102 = Power(
   Base=Divide(A=Saturate(None=@n101), B=OneMinus(None=@n100))
   Exp=ScalarParameter() {defaultValue=1, parameterName=SplineThickness_SpikeDataPower}
  )
  n103 = Multiply(A=@n28, B=Multiply(A=LinearInterpolate(A=@n98, B=@n99, Alpha=@n102), B=@n97.ScaleSplineWidth))
  n104 = VectorParameter() {
   group=WidthEdgeMask
   parameterName=WidthEdgeMask_EdgeMask_InvertControl
   channelNames {a="", b=Power, g=Multiply, r=Offset}
   defaultValue=<rgba 0 10 1 1>
  }
  n105 = VectorParameter() {
   group=WidthEdgeMask
   parameterName=WidthEdgeMask_EdgeMask_Control
   channelNames {a="", b=Power, g=Multiply, r=Offset}
   defaultValue=<rgba 0 10 1 1>
  }
  n106 = MaterialFunctionCall(
   UorV=MaterialFunctionCall(Float2=NamedRerouteUsage()) {materialFunction=&BreakOutFloat2Components}.G
   Invert_EdgeControl=@n104
   NoInvert_EdgeControl=@n105
  ) {materialFunction=&FXCommon_EdgeMask_Ex}
  n107 = Constant() {r=1}
  n108 = VectorParameter() {parameterName=WidthEdgeMask_Invert, defaultValue=<rgba 0 0.2 0 1>}
  n109 = MaterialFunctionCall(Input=OneMinus(None=@n106.Mask), Input Low=Constant(), Input High=@n107, Target Low=@n108.G, Target High=@n107) {materialFunction=&RemapValueRange}
  n110 = Constant() {r=-0.5}
  n111 = TextureCoordinate()
  n112 = Multiply(
   A=Multiply(A=Multiply(A=@n61, B=@n103), B=LinearInterpolate(A=@n106.Mask, B=@n109.Result, Alpha=@n108.R))
   B=ComponentMask(None=Add(A=@n110, B=@n111)) {r=true}
  )
  NamedRerouteDeclaration(None=Add(A=@n88, B=Multiply(A=@n95, B=@n112))) {name=WPO_SplineThickness, nodeColor=<rgba 0.4766 1 0 1>}
  n113 = TextureObjectParameter() {
   group=DissolveTexture
   parameterName=DissolveTexture_Texture
   samplerType=SAMPLERTYPE_LinearColor
   sortPriority=0
   texture=/Engine/Functions/Engine_MaterialFunctions02/ExampleContent/Textures/OffsetNoiseDistanceFields_ForNormals
  }
  n114 = VectorParameter() {
   group=DissolveTexture
   parameterName=DissolveTexture_UVControl
   channelNames {a=VPan, b=UPan, g=VScale, r=UScale}
   defaultValue=<rgba 0.5 1.5 0 0.1>
  }
  n115 = MaterialFunctionCall(Texture=@n113, UVControl_RGB=@n114, UVControl_A=@n114, "[Override]UVInput"=NamedRerouteUsage()) {materialFunction=&FXCommon_TextureSample}
  n116 = DynamicParameter() {
   parameterIndex=1
   defaultValue=<rgba 0 2 -1 0>
   paramNames [NewLightningTech_WPOStr, MoveAlignLightningDirSphereMaskPower, TextureDissolve, TextureWPONoise]
  }
  n117 = Subtract(A=@n116.TextureDissolve, B=Multiply(A=NamedRerouteUsage(), B=ScalarParameter() {defaultValue=0.3, parameterName=Dissolve_ScaleWidthFactor}))
  NamedRerouteDeclaration(None=MaterialFunctionCall(In=Subtract(A=@n115.R, B=@n117), Contrast=ScalarParameter() {defaultValue=10, parameterName=Dissolve_Contrast}) {materialFunction=&CheapContrast}.Result) {name=Dissolve, nodeColor=<rgba 0.2422 0 1 1>}
  n118 = TextureObjectParameter() {
   group=TextureWPONoise
   parameterName=TextureWPONoise_Texture
   sortPriority=0
   texture=/Game/Effects/Fort_Effects/Emotes/UltraLight/Texture/T_SoftNoise_BC7
  }
  n119 = VectorParameter() {
   group=TextureWPONoise
   parameterName=TextureWPONoise_UVControl
   channelNames {a=VPan, b=UPan, g=VScale, r=UScale}
   defaultValue=<rgba 0 1 0 0.5>
  }
  n120 = Add(
   A=Time() {bOverride_Period=true, period=6666}
   B=DynamicParameter() {paramNames [MaxGrowthLength, NewLightningTech_InvertDir, WPORand, NewLightningTech_Index]}.WPORand
  )
  n121 = NamedRerouteUsage()
  n122 = DynamicParameter() {
   parameterIndex=1
   defaultValue=<rgba 0 2 -1 0>
   paramNames [NewLightningTech_WPOStr, MoveAlignLightningDirSphereMaskPower, TextureDissolve, TextureWPONoise]
  }
  n123 = Multiply(
   A=Add(
    A=MaterialFunctionCall(Texture=@n118, UVControl_RGB=@n119, UVControl_A=@n119, "[Override]TimeInput"=@n120, "[Override]UVInput"=@n121) {materialFunction=&FXCommon_TextureSample}.RGB
    B=ScalarParameter() {defaultValue=-0.5, parameterName=TextureWPONoise_Offset}
   )
   B=Add(A=@n122.TextureWPONoise, B=ScalarParameter() {parameterName=TextureWPONoise_BaseNoiseIntensity})
  )
  n124 = VectorParameter() {group=TextureWPONoise, parameterName=TextureWPONoise_Mask, channelNames {a=VPan, b=UPan, g=VScale, r=UScale}, defaultValue=<rgba 1 1 1 0.5>}
  n125 = MaterialFunctionCall(Float2=NamedRerouteUsage()) {materialFunction=&BreakOutFloat2Components}
  n126 = VectorParameter() {
   group=TextureWPONoise
   parameterName=TextureWPONoise_EdgeMask_InvertControl
   channelNames {a="", b=Power, g=Multiply, r=Offset}
   defaultValue=<rgba 0 10000 1 1>
  }
  n127 = VectorParameter() {
   group=TextureWPONoise
   parameterName=TextureWPONoise_EdgeMask_Control
   channelNames {a="", b=Power, g=Multiply, r=Offset}
   defaultValue=<rgba 0 10000 1 1>
  }
  n128 = Transform(None=Multiply(
    A=Multiply(A=@n123, B=@n124)
    B=MaterialFunctionCall(UorV=@n125.G, Invert_EdgeControl=@n126, NoInvert_EdgeControl=@n127) {materialFunction=&FXCommon_EdgeMask_Ex}.Mask
   )) {transformSourceType=TRANSFORMSOURCE_Local}
  n129 = Constant3Vector()
  NamedRerouteDeclaration(None=MaterialFunctionCall("Xb1,PS4"=@n128, "PC-High&&Epic,XSX,PS5"=@n128, "PC-Perf,Android"=@n129, Switch=@n129, "PC-Low(NotPerf)&&Medium"=@n128) {materialFunction=&PlatformSwitch_3D}.Result) {name=TextureWPO, nodeColor=<rgba 1 0.7031 0 1>}
  n130 = TextureCoordinate()
  NamedRerouteDeclaration(None=StaticSwitchParameter(True=MaterialFunctionCall(XY=@n130) {materialFunction=&Swizzle}.YX, False=@n130) {parameterName=SwizzleUV}) {name=TUV, nodeColor=<rgba 0.8281 1 0 1>}
  n131 = TextureSampleParameter2DArray(UVs=MaterialFunctionCall(R=@n79, G=Constant() {r=0.5}, B=@n80) {materialFunction=&AppendMany}.RGB) {
   group=NewLightningTech
   parameterName=LightningTexture_ThicknessArray
   samplerType=SAMPLERTYPE_LinearColor
   texture=/Game/Effects/Fort_Effects/GameplayPulginsAssets/ToonLightning/GP29/Texture/T_GP29Lightning_ThicknessArray
  }
  NamedRerouteDeclaration(None=@n131.R) {name=NewLightningTech_ThickTech, nodeColor=<rgba 1 0.9141 0 1>}
  NamedRerouteDeclaration(None=@n87) {name=NewLightningTech_LSOffset, nodeColor=<rgba 1 0.375 0 1>}
  ScalarParameter() {defaultValue=-0.7, parameterName=Dissolve_Value}
  TextureCoordinate()
  TextureCoordinate()
  TextureCoordinate()
  TextureCoordinate()
  TextureCoordinate()
  TextureCoordinate()
 ]
}
